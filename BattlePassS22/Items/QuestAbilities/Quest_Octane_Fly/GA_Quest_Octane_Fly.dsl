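; Gameplay ability for the Season 22 Octane Fly quest, tracking how far the player travels airborne in a tagged vehicle.
; Vehicle enter starts a looping timer after a tag check; exit clears it and ends the ability.
; Each tick adds floored horizontal meters while airborne and resets the total on landing.
; Reaching the target distance in one flight sends the quest stat event and ends the ability.
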(event Custom|Check%20In%20Air
  (bind _owningplayerpawn (Variables|GAVars|Data|GetOwningPlayerPawn))
  (bind _returnvalue (Transformation|GetActorLocation _owningplayerpawn))
  (bind (_previouslocation_x _previouslocation_y) (Math|Vector|BreakVector (Variables|GetPreviousLocation_X)))
  (bind (_currentlocation_x _currentlocation_y) (Math|Vector|BreakVector (Variables|GetCurrentLocation_X)))
  ;[1 1 1] Check if Vehicle is in the air
  (Utilities|Casting|CastToFortAthenaVehicle (Vehicle|GetVehicleInterface (Variables|GAVars|Data|GetOwningPlayerPawn))
    (:then
      (if (Vehicle|VehicleIsinAir (Vehicle|GetVehicleInterface (Variables|GAVars|Data|GetOwningPlayerPawn)))
        ;[1 1 1] If in the air, calculate distance traveled while in the air
        (Variables|Default|SetCurrentLocation _returnvalue)
        (Variables|Default|SetMetersTraveled (+ (Math|Float|Floor (/ (Math|Vector|VectorLength (Math|Vector|vector-vector (Math|Vector|MakeVector _currentlocation_x _currentlocation_y) (Math|Vector|MakeVector _previouslocation_x _previouslocation_y))) 1e+02)) (Variables|Default|GetMetersTraveled)))
        (Variables|Default|SetPreviousLocation (Variables|Default|GetCurrentLocation))
        (if (>= (Variables|Default|GetMetersTraveled) (Variables|Default|GetDistanceForCompletion))
          (Quest|SendComplexCustomStatEvent (Variables|GAVars|Data|GetOwningPlayerQuestManager) 0 "(GameplayTags=((TagName=\"athena.quests.s22.Octane.Fly\")))" "(GameplayTags=)" 1)
          (Utilities|Time|ClearandInvalidateTimerbyHandle (Variables|Default|GetTimer%20Handle))
          (Ability|EndAbility))
        (else
          (Variables|Default|SetPreviousLocation _returnvalue)
          (Variables|Default|SetMetersTraveled 0))))
    (:CastFailed
      (Ability|EndAbility))))

(event Ability|EventActivateAbilityFromEvent
  (params (EventData :type Gameplay%20Event%20Data%20Structure%20%28by%20ref%29))
  (bind _timer_handle (Variables|Default|GetTimer%20Handle))
  (|CallParentFunction EventData)
  (switch Utilities|FlowControl|Switch|SwitchonGameplayTag (Utilities|Struct|BreakGameplayEventData EventData)
    (:Event.Vehicle.Entered
      ;[1 1 1] Check Vehicle Tags
      (if (GameplayTags|TagContainer|CompareTagAssetInterfacetoOtherTags (Variables|Default|GetT_Car) (Vehicle|GetVehicleAbilitySystemComponent (Vehicle|GetVehicleInterface (Variables|GAVars|Data|GetOwningPlayerPawn))))
        ;[1 1 1] Start Checking Vehicle If Vehicle is in Air
        (bind _returnvalue (Utilities|Time|SetTimerbyEvent (delegate Custom|Check%20In%20Air) (Variables|Default|GetExecutionTimeInterval) true))
        (Variables|Default|SetTimer%20Handle _returnvalue)
        (else
          (Ability|EndAbility))))
    (:Event.Vehicle.Exit
      ;[1 1 1] On Vehicle Exit Clear Timer and End Ability
      (if (Utilities|Time|IsValidTimerHandle _timer_handle)
        (Utilities|Time|ClearandInvalidateTimerbyHandle _timer_handle)
        (else
          _))
      (Ability|EndAbility))))
